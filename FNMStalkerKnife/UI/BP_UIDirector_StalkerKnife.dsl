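; UI director for the Stalker Knife that shows a tac sprint UI scene while the viewed pawn holds the Mantis knife weapon.
; At BeginPlay, except on mobile with touch, it binds to the player controller's view target changes.
; Each change moves the weapon-equipped binding to the new FortPawn.
; Equip events cast the weapon to the knife class, and a delta gate adds or removes the scene only when that state changes.
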(event EventBeginPlay
  (if (not (and (TouchPlatform|IsMobileGame) (TouchPlatform|IsUsingTouch 0)))
    (bind _asfort_player_controller (Utilities|Casting|CastToFortPlayerController (DynamicUIDirector|GetOwningLocalPlayerController)))
    (EventDispatchers|BindEventtoOnViewTargetChanged _asfort_player_controller (EventDispatchers|CreateEvent :!event OnViewTargetChanged))
    (CallFunction|OnViewTargetChanged 0 0 (Pawn|GetViewTarget _asfort_player_controller))))

(fn OnViewTargetChanged
  (params (PlayerController :type Fort%20Player%20Controller%20Object%20Reference) (OldViewTarget :type Actor%20Object%20Reference) (NewViewTarget :type Actor%20Object%20Reference))
  (bind _outputdelegate (EventDispatchers|CreateEvent :!event OnWeaponEquipped))
  (bind _boundfortpawn (Variables|Default|GetBoundFortPawn))
  (Utilities|IsValid _boundfortpawn
    (:Is%20Valid
      (FortPawn|UnbindEventfromOnWeaponEquippedDelegate _boundfortpawn _outputdelegate))
    (:Is%20Not%20Valid
      _))
  (bind _asfort_pawn (Utilities|Casting|CastToFortPawn NewViewTarget)
    (:then
      (bind _output_get (Variables|Default|SetBoundFortPawn _asfort_pawn))
      (FortPawn|BindEventtoOnWeaponEquippedDelegate _output_get _outputdelegate))
    (:CastFailed
      (Variables|Default|SetBoundFortPawn 0))))

(fn OnWeaponEquipped
  (params (NewWeapon :type Fort%20Weapon%20Object%20Reference) (PrevWeapon :type Fort%20Weapon%20Object%20Reference))
  (Utilities|Casting|CastToB_StalkerKnife_MantisWeapon NewWeapon
    (:then
      (CallFunction|SetTacSprintSceneActive true))
    (:CastFailed
      (CallFunction|SetTacSprintSceneActive false))))

(fn SetTacSprintSceneActive
  (params (IsActive :type Boolean))
  (bind _tacsprintscene (Variables|Default|GetTacSprintScene))
  (bind _updated_variable (Utilities|FlowControl|DeltaGateWildcard (Variables|Default|GetIsTacSprintActive) IsActive)
    (:Success
      (if _updated_variable
        (DynamicUIDirector|AddScene _tacsprintscene)
        (else
          (DynamicUIDirector|RemoveScene _tacsprintscene))))))

(fn ConstructionScript)
