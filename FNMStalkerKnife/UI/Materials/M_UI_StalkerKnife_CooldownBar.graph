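# M_UI_StalkerKnife_CooldownBar is a translucent UI material that draws the Stalker Knife's radial cooldown bar.
# MF_UI_RadialArc calls on rotated UVs, shaped by start and end angles, cut arcs from a thin SDF circle ring, with an inverted Progress parameter for fill.
# A dimmed constant-fill arc and a variable-fill arc combine with a textured frame mask under an Opacity parameter, while FrameColor tints the result.
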Material /FNMStalkerKnife/UI/Materials/M_UI_StalkerKnife_CooldownBar
graph {
 properties {blendMode=BLEND_Translucent, materialDomain=MD_UI, pixelDepthOffsetMode=PDOM_Legacy, usedShadingModels=Default Lit}
 nodes [
  n1 = NamedRerouteUsage()
  Composite!(Fill Ratio=@n1) {subgraphName=Overheat Progress Mask}
  n2 = Constant()
  Composite(Fill Ratio=@n2) {subgraphName=Overheat Progress}
  n3 = NamedRerouteUsage()
  Composite(Fill Ratio=@n3) {subgraphName=Overheat Progress}
  n4 = MaterialFunctionCall(UVs=NamedRerouteUsage(), "Rotation Angle (0-1)"=Constant() {r=0.5}) {materialFunction=&CustomRotator}
  n5 = ScalarParameter() {defaultValue=0.97, parameterName=Start Angle, sliderMax=1, sortPriority=31}
  n6 = ScalarParameter() {defaultValue=0.25, parameterName=End Angle, sliderMax=1}
  MaterialExpressionReroute_4 = Reroute(None=@n2)
  n7 = Multiply(
   A=NamedRerouteUsage()
   B=SmoothStep(Value=MaterialFunctionCall!(
     UV=@n4.Rotated Values
     "Arc Start Angle (0-1)"=@n5
     "Arc End Angle (0-1)"=@n6
     "Fill Ratio (0-1)"=@MaterialExpressionReroute_4.Fill Ratio
     Fill Clockwise
    ) {materialFunction=/FortUILibrary/MaterialFunctions/Basics/MF_UI_RadialArc}.Radial Fill) {constMax=0.001}
  )
  MaterialExpressionReroute_5 = Reroute(None=@n7)
  n8 = Multiply(A=NamedRerouteUsage(), B=Multiply(A=@MaterialExpressionReroute_5.Output) {constB=0.3})
  n9 = MaterialFunctionCall(UVs=NamedRerouteUsage(), "Rotation Angle (0-1)"=Constant() {r=0.5}) {materialFunction=&CustomRotator}
  n10 = ScalarParameter() {defaultValue=0.97, parameterName=Start Angle, sliderMax=1, sortPriority=31}
  n11 = ScalarParameter() {defaultValue=0.25, parameterName=End Angle, sliderMax=1}
  MaterialExpressionReroute_2 = Reroute(None=@n3)
  n12 = Multiply(
   A=NamedRerouteUsage()
   B=SmoothStep(Value=MaterialFunctionCall(UV=@n9.Rotated Values, "Arc Start Angle (0-1)"=@n10, "Arc End Angle (0-1)"=@n11, "Fill Ratio (0-1)"=@MaterialExpressionReroute_2.Fill Ratio) {materialFunction=&MF_UI_RadialArc}.Radial Fill) {constMax=0.001}
  )
  MaterialExpressionReroute_3 = Reroute(None=@n12)
  n13 = Multiply(A=NamedRerouteUsage(), B=Multiply(A=@MaterialExpressionReroute_3.Output))
  Multiply(A=Max(A=NamedRerouteUsage(), B=Add(A=@n8, B=@n13)), B=ScalarParameter() {defaultValue=1, parameterName=Opacity})
  NamedRerouteDeclaration(None=MaterialFunctionCall() {materialFunction=&GetUserInterfaceUV}.Normalized UV) {name=UV, nodeColor=<rgba 0 0 0 1>}
  n14 = MaterialFunctionCall(OutlineThickness=ScalarParameter() {defaultValue=0.01, group=AmmoCounter, parameterName=RadialThickness, sliderMax=1}) {materialFunction=&MF_UI_Circle}
  NamedRerouteDeclaration(None=OneMinus(None=SmoothStep(Value=@n14.SDFOutline) {constMax=0.01})) {name=Mask_Radial, nodeColor=<rgba 0 0 0 1>}
  n15 = TextureCoordinate()
  n16 = Constant2Vector() {g=-0.12, r=-0.146}
  NamedRerouteDeclaration(None=MaterialFunctionCall(UVs=Add(A=@n15, B=@n16), Texture Scale=Constant() {r=0.36}) {materialFunction=&ScaleUVsByCenter}.UVs) {name=TextureUVs, nodeColor=<rgba 0 0 0 1>}
  n17 = MaterialFunctionCall(UVs=NamedRerouteUsage(), "Rotation Angle (0-1)"=Constant() {r=0.5}) {materialFunction=&CustomRotator}
  n18 = ScalarParameter() {defaultValue=0.97, parameterName=Start Angle, sliderMax=1, sortPriority=31}
  n19 = ScalarParameter() {defaultValue=0.25, parameterName=End Angle, sliderMax=1}
  MaterialExpressionReroute_6 = Reroute(None=@n1)
  n20 = Multiply(
   A=NamedRerouteUsage()
   B=SmoothStep(Value=MaterialFunctionCall(UV=@n17.Rotated Values, "Arc Start Angle (0-1)"=@n18, "Arc End Angle (0-1)"=@n19, "Fill Ratio (0-1)"=@MaterialExpressionReroute_6.Fill Ratio) {materialFunction=&MF_UI_RadialArc}.Radial Fill) {constMax=0.001}
  )
  MaterialExpressionReroute_7 = Reroute(None=@n20)
  n21 = VectorParameter() {parameterName=FrameColor, defaultValue=<rgba 1 1 1 1>}
  n22 = NamedRerouteUsage()
  NamedRerouteDeclaration(None=Add(A=Multiply(A=@MaterialExpressionReroute_7.Output, B=@n21.RGB), B=Multiply(A=@n21.RGB, B=@n22))) {name=Final Color, nodeColor=<rgba 0.8984 0 1 1>}
  n23 = TextureSample(UVs=NamedRerouteUsage()) {texture=/FNMStalkerKnife/UI/Textures/T_UI_StarlkerKnife_Cooldown}
  NamedRerouteDeclaration(None=Multiply(A=SmoothStep(Value=@n23.R) {constMax=0.4, constMin=0.35}) {constB=0.9}) {name=OuterFrame_Mask, nodeColor=<rgba 0 0 0 1>}
  NamedRerouteDeclaration(None=OneMinus(None=ScalarParameter() {defaultValue=0.5, group=Color, parameterName=Progress, sliderMax=1})) {name=Progress, nodeColor=<rgba 0 0 0 1>}
  NamedRerouteUsage()
  PinBase() {reroutePins [Fill Ratio {expression=&self:MaterialGraph_0.MaterialGraphNode_Composite_2.Overheat Progress Mask.MaterialExpressionReroute_6}]}
  PinBase!(Output=@n20) {reroutePins [Output {expression=&self:MaterialGraph_0.MaterialGraphNode_Composite_2.Overheat Progress Mask.MaterialExpressionReroute_7}]}
  PinBase!() {reroutePins [Fill Ratio {expression=&self:MaterialGraph_0.MaterialGraphNode_Composite_3.Overheat Progress.MaterialExpressionReroute_4}]}
  PinBase!(Output=@n7) {reroutePins [Output {expression=&self:MaterialGraph_0.MaterialGraphNode_Composite_3.Overheat Progress.MaterialExpressionReroute_5}]}
  PinBase!() {reroutePins [Fill Ratio {expression=&self:MaterialExpressionReroute_2}]}
  PinBase!(Output=@n12) {reroutePins [Output {expression=&self:MaterialExpressionReroute_3}]}
 ]
}
